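# Niagara vehicle boost cosmetic that trails colored ribbons and spinning shape sprites from the vehicle's skeletal-mesh boost sockets.
# CrazyShapes merges two spawn groups into one emitter, scattering curl-noise-driven SubUV sprites across the ports through two renderers.
# MainRibbon spawns one ribbon per port, layered through five renderers with separate colors and width curves; a duplicate ribbon emitter is disabled.
# User.bIsPainted switches between painted-variant colors and the default multicolor palette.
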NiagaraSystem /VehicleCosmetics/Boosts/Boost_CrazyShapes/FX/NS_Boost_CrazyShapes
dependencies {
 usedDataInterfaces [
  &NiagaraDataInterfaceArrayColor
  &NiagaraDataInterfaceCurve
  &NiagaraDataInterfaceSpriteRendererInfo
  &NiagaraDataInterfaceArrayFloat
  &NiagaraDataInterfaceSkeletalMesh
 ]
 usedDynamicInputs []
 usedModules [
  &self:SystemSpawnScript.SystemScriptSource.SystemScriptGraph.NiagaraNodeAssignment_0.SetVariables_#1
  &SystemState
 ]
}
emitters {
 CrazyShapes {
  emitterUpdateScript [
   &EmitterState()
   &VC_FilteredSocketCount()
   &SpawnRate(SpawnRate=20, Use Spawn Probability=false)
   &SpawnRate(SpawnGroup=1, SpawnRate=20, Use Spawn Probability=false) {module=SpawnRate001}
  ]
  particleSpawnScript [
   /SetVariables(Particles.VisibilityTag=$Emitter.SpawnGroup) {isSetParameters=true}
   &VC_PaintedVariantsOverride(
    Black=<rgba 0.039 0.039 0.039 1>
    BurntSienna=<rgba 0.09 0.0276 0.018 1>
    Cobalt=<rgba 0.025 0.05 0.65 1>
    Crimson=<rgba 0.4635 0 0.0126 1>
    Orange=<rgba 0.6823 0.0913 0.0082 1>
    Pink=<rgba 0.7552 0.1495 0.6551 1>
    Purple=<rgba 0.2101 0.0784 0.56 1>
    SkyBlue=<rgba 0.1355 0.4637 0.7031 1>
   )
   &InitializeParticle(
    Color=$Particles.Color
    Lifetime=/VehicleCosmetics/FX/Niagara/Modules/MergeEmitters/VC_FloatBySpawnGroup(
     SpawnGroup or ExecIndex:NiagaraInt32=0
     CustomInt:NiagaraInt32=0
     MaximumSpawnGroups:NiagaraInt32=2
     Float SG 0:NiagaraFloat=&RandomRangeFloat(Minimum=0.4, Maximum=0.7)
     Float SG 1:NiagaraFloat=&RandomRangeFloat(Minimum=0.5, Maximum=1)
     Float SG 2:NiagaraFloat=0
     Float SG 3:NiagaraFloat=0
     Float SG 4:NiagaraFloat=0
     Float SG 5:NiagaraFloat=0
    )
    Sprite Rotation Mode=::NewEnumerator2
    Sprite Size Mode=::NewEnumerator3
    Uniform Sprite Size=&VC_FloatBySpawnGroup(
     SpawnGroup or ExecIndex:NiagaraInt32=0
     CustomInt:NiagaraInt32=0
     MaximumSpawnGroups:NiagaraInt32=2
     Float SG 0:NiagaraFloat=&RandomRangeFloat(Minimum=40, Maximum=55)
     Float SG 1:NiagaraFloat=&RandomRangeFloat(Minimum=30, Maximum=50)
     Float SG 2:NiagaraFloat=0
     Float SG 3:NiagaraFloat=0
     Float SG 4:NiagaraFloat=0
     Float SG 5:NiagaraFloat=0
    )
    UsePositionOffset=false
    Write Material Random=true
   )
   &PartitionParticles(Number Of Partitions=$Emitter.PortCount)
   &SkeletalMeshLocation(
    Bone / Socket Index=$Particles.PartitionParticles.Partition
    Bone / Socket Sampling Mode=::NewEnumerator5
    MaskSpawnGroup=false
    Mesh Sampling Type=::NewEnumerator6
    Sampled Position Offset=/VehicleCosmetics/FX/Niagara/Modules/MergeEmitters/VC_VectorBySpawnGroup(
     SpawnGroup or ExecIndex:NiagaraInt32=0
     CustomInt:NiagaraInt32=0
     MaximumSpawnGroups:NiagaraInt32=2
     Vector Group 0:Vector3f=<-40 0 0>
     Vector Group 1:Vector3f=<0 0 0>
     Vector Group 2:Vector3f=<0 0 0>
     Vector Group 3:Vector3f=<0 0 0>
     Vector Group 4:Vector3f=<0 0 0>
     Vector Group 5:Vector3f=<0 0 0>
    )
    Skeletal Mesh=$System.SkeletalMesh
   )
   &AddVelocity(
    Velocity=&VC_VectorBySpawnGroup(
     SpawnGroup or ExecIndex:NiagaraInt32=0
     CustomInt:NiagaraInt32=0
     MaximumSpawnGroups:NiagaraInt32=2
     Vector Group 0:Vector3f=&RandomRangeVector(Minimum=<-100 0 0>, Maximum=<-200 0 0>)
     Vector Group 1:Vector3f=&RandomRangeVector(Minimum=<-500 0 0>, Maximum=<-1000 0 0>)
     Vector Group 2:Vector3f=<0 0 0>
     Vector Group 3:Vector3f=<0 0 0>
     Vector Group 4:Vector3f=<0 0 0>
     Vector Group 5:Vector3f=<0 0 0>
    )
    Velocity Speed Scale=0.6
   )
   &ShapeLocation(MaskSpawnGroup=false, Override Seed=false, Sphere Radius=40, UseOwnerScale=true)
   /SetVariables(
    Particles.ColorTag:LinearColor=&MakeCustomLinearColorFromBool(
     Bool=$User.bIsPainted
     True Color=$Particles.Color
     False Color=&SelectLinearColorFromArray(
      Array Sampling Mode=::NewEnumerator0
      Direct Array Index=?
      Color Selection Array=DataInterface() {propertyValues {ColorData [<rgba 1 1 0 1>, <rgba 1 0.1195 0 1>, <rgba 0.3005 0.5395 1 1>, <rgba 0.2502 0.1301 0.4508 1>]}}
      Random Seed=0
     )
    )
    Particles.ColorTag1:LinearColor=&MakeCustomLinearColorFromBool(
     Bool=$User.bIsPainted
     True Color=$Particles.Color
     False Color=&SelectLinearColorFromArray(
      Array Sampling Mode=::NewEnumerator0
      Direct Array Index=?
      Color Selection Array=DataInterface() {propertyValues {ColorData [<rgba 1 1 0 1>, <rgba 0.3005 0.5395 1 1>, <rgba 0.2502 0.1301 0.4508 1>]}}
      Random Seed=0
     )
    )
   ) {isSetParameters=true}
  ]
  particleUpdateScript [
   &ParticleState()
   &ScaleSpriteSize(Uniform Curve Sprite Scale=DataInterface() {propertyValues {Curve {
       defaultValue=1
       keys [
        <at 0 0 RCIM_Cubic>
        {arriveTangentWeight=0.1, interpMode=RCIM_Cubic, leaveTangentWeight=0.1333, tangentMode=RCTM_User, time=0.3, value=1}
        {arriveTangentWeight=0.2325, interpMode=RCIM_Cubic, leaveTangentWeight=0.1744, tangentMode=RCTM_User, time=0.7, value=1}
        {interpMode=RCIM_Cubic, tangentMode=RCTM_User, time=1, value=0}
       ]
      }}})
   &SpriteRotationRate(Rotation Rate=&RandomRangeFloat(Minimum=100, Maximum=400))
   &CurlNoiseForce(Bias Noise=false, Noise Frequency=10, Noise Strength=500, Pan Noise=false)
   &SolveForcesAndVelocity(Clamp Velocity=false, Limit Acceleration=false)
   &SubUVAnimation(
    Direct Frame Index:NiagaraFloat=&RandomRangeFloat(Minimum=0, Maximum=4)
    "Sprite Renderer "=DataInterface() {propertyValues {SpriteRenderer=&self:CrazyShapes_0.NiagaraSpriteRendererProperties_0}}
    SubUV Animation Mode:ENiagara_SubUVLookupModeV2=::NewEnumerator5
   )
   /SetVariables(
    Particles.SpriteColor0:LinearColor=&MultiplyLinearColor(
     Linear Color=$Particles.Initial.ColorTag
     Scale Factor=/Niagara/DynamicInputs/TypeConversions/MakeCustomVector4FromBool(
      Bool:NiagaraBool=$User.bIsPainted
      True Vector4:Vector4f=&MakeVector4FromVectorAndFloat(
       "Vector (RGB)"=&VectorFromFloat(Value=&SelectFloatFromArray(
         Array Sampling Mode=::NewEnumerator0
         Direct Array Index=?
         Float Selection Array=DataInterface() {propertyValues {FloatData [0.5,1.3]}}
         Random Seed=0
        ))
       "Float (Alpha)"=1
      )
      False Vector4:Vector4f=&MakeVector4FromVectorAndFloat("Vector (RGB)"=<1 1 1>, "Float (Alpha)"=1)
     )
    )
    Particles.SpriteColor1:LinearColor=&MultiplyLinearColor(
     Linear Color=$Particles.Initial.ColorTag1
     Scale Factor=&MakeCustomVector4FromBool(
      Bool:NiagaraBool=$User.bIsPainted
      True Vector4:Vector4f=&MakeVector4FromVectorAndFloat(
       "Vector (RGB)"=&VectorFromFloat(Value=&SelectFloatFromArray(
         Array Sampling Mode=::NewEnumerator0
         Direct Array Index=?
         Float Selection Array=DataInterface() {propertyValues {FloatData [0.5,1.3]}}
         Random Seed=0
        ))
       "Float (Alpha)"=1
      )
      False Vector4:Vector4f=&MakeVector4FromVectorAndFloat("Vector (RGB)"=<1 1 1>, "Float (Alpha)"=1)
     )
    )
   ) {isSetParameters=true}
  ]
  rendererClasses [&NiagaraSpriteRendererProperties]
  renderers [
   {rendererClass=&NiagaraSpriteRendererProperties}
   {rendererClass=&NiagaraSpriteRendererProperties, rendererIndex=1}
  ]
 }
 MainRibbon {
  emitterUpdateScript [
   &EmitterState()
   &VC_FilteredSocketCount()
   &VC_RateQualityHandler()
   &VC_SpawnRibbon_v1(Number of Ribbons=$Emitter.PortCount, SpawnRate=$Emitter.Spawn Rate)
  ]
  particleSpawnScript [
   &VC_PaintedVariantsOverride(
    Black=<rgba 0.036 0.036 0.036 1>
    BurntSienna=<rgba 0.085 0.0306 0.0213 1>
    Cobalt=<rgba 0.031 0.0539 0.72 1>
    Crimson=<rgba 0.6 0.024 0.0344 1>
    Grey=<rgba 0.2135 0.2135 0.2135 1>
    Lime=<rgba 0.4323 0.8646 0 1>
    Orange=<rgba 0.6823 0.0913 0.0082 1>
    Pink=<rgba 0.9219 0.2161 0.8052 1>
    Purple=<rgba 0.2 0.075 0.7 1>
    R1 Black:LinearColor=<rgba 0 0 0 1>
    R1 BurntSienna:LinearColor=<rgba 0.04 0.0175 0.0136 1>
    R1 Cobalt:LinearColor=<rgba 0.0175 0.0286 0.35 1>
    R1 Crimson:LinearColor=<rgba 0.1979 0 0.0086 1>
    R1 ForestGreen:LinearColor=<rgba 0 0.1615 0.0082 1>
    R1 Grey:LinearColor=<rgba 0.13 0.13 0.13 1>
    R1 Lime:LinearColor=<rgba 0.099 0.1979 0 1>
    R1 Orange:LinearColor=<rgba 0.4479 0.0372 0.0057 1>
    R1 Pink:LinearColor=<rgba 0.4531 0.1251 0.3493 1>
    R1 Purple:LinearColor=<rgba 0.0508 0.0247 0.22 1>
    R1 Saffron:LinearColor=<rgba 0.35 0.3062 0.035 1>
    R1 SkyBlue:LinearColor=<rgba 0.0059 0.1703 0.4062 1>
    R1 TitaniumWhite:LinearColor=<rgba 0.4323 0.4323 0.4323 1>
    R2 Black:LinearColor=<rgba 0.036 0.036 0.036 1>
    R2 BurntSienna:LinearColor=<rgba 0.12 0.048 0.0336 1>
    R2 Cobalt:LinearColor=<rgba 0.08 0.104 0.8 1>
    R2 Crimson:LinearColor=<rgba 0.6 0.024 0.0344 1>
    R2 Grey:LinearColor=<rgba 0.2135 0.2135 0.2135 1>
    R2 Lime:LinearColor=<rgba 0.4323 0.8646 0 1>
    R2 Orange:LinearColor=<rgba 0.6823 0.0913 0.0082 1>
    R2 Pink:LinearColor=<rgba 0.9219 0.2161 0.8052 1>
    R2 Purple:LinearColor=<rgba 0.34 0.14 0.7 1>
    R2 Saffron:LinearColor=<rgba 0.95 0.8 0.057 1>
    R2 SkyBlue:LinearColor=<rgba 0.1355 0.4637 0.7031 1>
    R3 Black:LinearColor=<rgba 0.009 0.009 0.009 1>
    R3 BurntSienna:LinearColor=<rgba 0.03 0.0159 0.0135 1>
    R3 Cobalt:LinearColor=<rgba 0.02 0.036 0.32 1>
    R3 Crimson:LinearColor=<rgba 0.1979 0 0.0086 1>
    R3 ForestGreen:LinearColor=<rgba 0 0.1615 0.0082 1>
    R3 Grey:LinearColor=<rgba 0.13 0.13 0.13 1>
    R3 Lime:LinearColor=<rgba 0.099 0.1979 0 1>
    R3 Orange:LinearColor=<rgba 0.4479 0.0372 0.0057 1>
    R3 Pink:LinearColor=<rgba 0.4531 0.1251 0.3493 1>
    R3 Purple:LinearColor=<rgba 0.0853 0.0349 0.31 1>
    R3 Saffron:LinearColor=<rgba 0.35 0.3062 0.035 1>
    R3 SkyBlue:LinearColor=<rgba 0.0059 0.1703 0.4062 1>
    R3 TitaniumWhite:LinearColor=<rgba 0.4323 0.4323 0.4323 1>
    Saffron=<rgba 0.95 0.8 0.057 1>
    SkyBlue=<rgba 0.1355 0.4637 0.7031 1>
    Use R1 Painted Overrides=true
    Use R2 Painted Overrides=true
    Use R3 Painted Overrides=true
   )
   &VC_InitializeRibbon_v1(Color=$Particles.Color, Enable Skeletal Position=true, Lifetime=0.8, Ribbon Width=70)
   &AddVelocity(Velocity=<-1400 0 0>) {module=AddVelocity001}
   /SetVariables(Particles.WidthTag1=40, Particles.WidthTag2=90, Particles.WidthTag3=30) {isSetParameters=true, module=SetVariables_#2}
  ]
  particleUpdateScript [
   &ParticleState()
   &SolveForcesAndVelocity(Clamp Velocity=false, Limit Acceleration=false)
   &VC_UpdateRibbon_v1(Enable Skeletal Position=true)
   &DynamicMaterialParameters(
    Index 0 Param 1=1
    Module.Index0Float0Write=true
    Module.Index0Float1Write=false
    Module.Index0Float2Write=false
    Module.Index0Float3Write=false
   )
   &ScaleRibbonWidth(Ribbon Width Scale=&FloatFromCurve(FloatCurve=DataInterface() {propertyValues {Curve {keys [
         <at 0 0.6 RCIM_Cubic>
         {arriveTangentWeight=0.1166, interpMode=RCIM_Cubic, leaveTangentWeight=0.2721, tangentMode=RCTM_User, time=0.3, value=1}
         <at 1 1>
        ]}}}))
   /SetVariables(
    Particles.RibbonColor0:LinearColor=&MultiplyLinearColor(
     Linear Color=&MakeCustomLinearColorFromBool(Bool=$User.bIsPainted, True Color=$Particles.Color, False Color=<rgba 0.3005 0.5395 1 1>)
     Scale Factor=&MakeVector4FromVectorAndFloat("Vector (RGB)"=<1 1 1>, "Float (Alpha)"=1)
    )
    Particles.RibbonColor1:LinearColor=&MultiplyLinearColor(
     Linear Color=&MakeCustomLinearColorFromBool(Bool=$User.bIsPainted, True Color=$Particles.R1 Color, False Color=<rgba 0.2502 0.1301 0.4508 1>)
     Scale Factor=&MakeVector4FromVectorAndFloat("Vector (RGB)"=<1 1 1>, "Float (Alpha)"=1)
    )
    Particles.RibbonColor2:LinearColor=&MultiplyLinearColor(
     Linear Color=$Particles.Color
     Scale Factor=&MakeVector4FromVectorAndFloat("Vector (RGB)"=<0 0 0>, "Float (Alpha)"=1)
    )
    Particles.RibbonColor3:LinearColor=&MultiplyLinearColor(
     Linear Color=&MakeCustomLinearColorFromBool(Bool=$User.bIsPainted, True Color=$Particles.R2 Color, False Color=<rgba 1 1 0 1>)
     Scale Factor=&MakeVector4FromVectorAndFloat("Vector (RGB)"=<1 1 1>, "Float (Alpha)"=1)
    )
    Particles.RibbonColor4:LinearColor=&MultiplyLinearColor(
     Linear Color=&MakeCustomLinearColorFromBool(Bool=$User.bIsPainted, True Color=$Particles.R3 Color, False Color=<rgba 1 0.12 0 1>)
     Scale Factor=&MakeVector4FromVectorAndFloat("Vector (RGB)"=<1 1 1>, "Float (Alpha)"=1)
    )
    Particles.RibbonWidth1:NiagaraFloat=&Multiply_Float(
     A=$Particles.WidthTag1
     B=&FloatFromCurve(FloatCurve=DataInterface() {propertyValues {Curve {keys [
          {interpMode=RCIM_Cubic, leaveTangentWeight=0.1, tangentMode=RCTM_User, time=0, value=0.6}
          {arriveTangentWeight=0.1166, interpMode=RCIM_Cubic, leaveTangentWeight=0.2721, tangentMode=RCTM_User, time=0.3, value=1}
          <at 1 1>
         ]}}})
    )
    Particles.RibbonWidth2:NiagaraFloat=&Multiply_Float(
     A=$Particles.WidthTag2
     B=&FloatFromCurve(FloatCurve=DataInterface() {propertyValues {Curve {keys [
          {interpMode=RCIM_Cubic, leaveTangentWeight=0.1, tangentMode=RCTM_User, time=0, value=0.6}
          {arriveTangentWeight=0.1166, interpMode=RCIM_Cubic, leaveTangentWeight=0.2721, tangentMode=RCTM_User, time=0.3, value=1}
          <at 1 1>
         ]}}})
    )
    Particles.RibbonWidth3:NiagaraFloat=&Multiply_Float(
     A=$Particles.WidthTag3
     B=&FloatFromCurve(FloatCurve=DataInterface() {propertyValues {Curve {keys [
          <at 0.085 0 RCIM_Constant>
          {arriveTangent=5.7357, interpMode=RCIM_Cubic, leaveTangent=5.7357, time=0.1158, value=0.3844}
          {arriveTangentWeight=0.1213, interpMode=RCIM_Cubic, leaveTangentWeight=0.2163, tangentMode=RCTM_User, time=0.2593, value=1}
          {arriveTangentWeight=0.0876, interpMode=RCIM_Cubic, leaveTangentWeight=0.3393, tangentMode=RCTM_User, time=0.5, value=1}
          {arriveTangent=-3.5379, interpMode=RCIM_Cubic, leaveTangent=-3.5379, time=0.7679, value=0.2867}
          <at 0.7827 0 RCIM_Constant>
         ]}}})
    )
   ) {isSetParameters=true, module=SetVariables_#3}
  ]
  rendererClasses [&NiagaraRibbonRendererProperties]
  renderers [
   {rendererClass=&NiagaraRibbonRendererProperties}
   {rendererClass=&NiagaraRibbonRendererProperties, rendererIndex=1}
   {rendererClass=&NiagaraRibbonRendererProperties, rendererIndex=2}
   {rendererClass=&NiagaraRibbonRendererProperties, rendererIndex=3}
   {rendererClass=&NiagaraRibbonRendererProperties, rendererIndex=4}
  ]
 }
 MainRibbon001 {
  bEnabled=false
  emitterUpdateScript [
   &EmitterState()
   &VC_FilteredSocketCount()
   &VC_RateQualityHandler()
   &VC_SpawnRibbon_v1(Number of Ribbons=$Emitter.PortCount, SpawnRate=$Emitter.Spawn Rate)
  ]
  particleSpawnScript [
   &VC_PaintedVariantsOverride(
    Black=<rgba 0.0249 0.0249 0.026 1>
    BurntSienna=<rgba 0.0723 0.0264 0.0166 1>
    Cobalt=<rgba 0.043 0.1242 0.5156 1>
    Crimson=<rgba 0.6 0.024 0.0344 1>
    Grey=<rgba 0.2135 0.2135 0.2135 1>
    Lime=<rgba 0.4323 0.8646 0 1>
    Orange=<rgba 0.6823 0.0913 0.0082 1>
    Pink=<rgba 0.9219 0.2161 0.8052 1>
    Purple=<rgba 0.2018 0.0754 0.5365 1>
    R1 Black:LinearColor=<rgba 0.015 0.015 0.015 1>
    R1 BurntSienna:LinearColor=<rgba 0.0312 0.0109 0.0061 1>
    R1 Cobalt:LinearColor=<rgba 0.0253 0.0648 0.2552 1>
    R1 Crimson:LinearColor=<rgba 0.1979 0 0.0086 1>
    R1 ForestGreen:LinearColor=<rgba 0 0.1615 0.0082 1>
    R1 Grey:LinearColor=<rgba 0.13 0.13 0.13 1>
    R1 Lime:LinearColor=<rgba 0.099 0.1979 0 1>
    R1 Orange:LinearColor=<rgba 0.4479 0.0372 0.0057 1>
    R1 Pink:LinearColor=<rgba 0.4531 0.1251 0.3493 1>
    R1 Purple:LinearColor=<rgba 0.085 0.032 0.3229 1>
    R1 Saffron:LinearColor=<rgba 0.35 0.3062 0.035 1>
    R1 SkyBlue:LinearColor=<rgba 0.0059 0.1703 0.4062 1>
    R1 TitaniumWhite:LinearColor=<rgba 0.6823 0.6823 0.6823 1>
    Saffron=<rgba 0.95 0.8 0.057 1>
    SkyBlue=<rgba 0.1355 0.4637 0.7031 1>
    TitaniumWhite=<rgba 0.0208 0.0208 0.0208 1>
    Use R1 Painted Overrides=true
   )
   &VC_InitializeRibbon_v1(Color=$Particles.Color, Enable Skeletal Position=true, Lifetime=0.8, Ribbon Width=70)
   &AddVelocity(Velocity=<-1400 0 0>) {module=AddVelocity001}
   /SetVariables(Particles.WidthTag1=40, Particles.WidthTag2=90, Particles.WidthTag3=30) {isSetParameters=true, module=SetVariables_#2}
  ]
  particleUpdateScript [
   &ParticleState()
   &SolveForcesAndVelocity(Clamp Velocity=false, Limit Acceleration=false)
   &VC_UpdateRibbon_v1(Enable Skeletal Position=true)
   &DynamicMaterialParameters(
    Index 0 Param 1=1
    Module.Index0Float0Write=true
    Module.Index0Float1Write=false
    Module.Index0Float2Write=false
    Module.Index0Float3Write=false
   )
   &ScaleRibbonWidth(Ribbon Width Scale=&FloatFromCurve(FloatCurve=DataInterface() {propertyValues {Curve {keys [
         <at 0 0.6 RCIM_Cubic>
         {arriveTangentWeight=0.1166, interpMode=RCIM_Cubic, leaveTangentWeight=0.2721, tangentMode=RCTM_User, time=0.3, value=1}
         <at 1 1>
        ]}}}))
   /SetVariables(
    Particles.RibbonColor0:LinearColor=&VC_DynamicScaleColor(
     Color to Scale:LinearColor=&MakeCustomLinearColorFromBool(Bool=$User.bIsPainted, True Color=$Particles.Color, False Color=<rgba 0.3005 0.5395 1 1>)
     RGB Scale:Vector3f=&VectorFromFloat(Value=1)
     Alpha Scale:NiagaraFloat=1
    )
    Particles.RibbonColor1:LinearColor=&VC_DynamicScaleColor(
     Color to Scale:LinearColor=&MakeCustomLinearColorFromBool(Bool=$User.bIsPainted, True Color=$Particles.R1 Color, False Color=<rgba 0.2502 0.1301 0.4508 1>)
     RGB Scale:Vector3f=&VectorFromFloat(Value=1)
     Alpha Scale:NiagaraFloat=1
    )
    Particles.RibbonColor2:LinearColor=&VC_DynamicScaleColor(
     Color to Scale:LinearColor=$Particles.Color
     RGB Scale:Vector3f=&VectorFromFloat(Value=0)
     Alpha Scale:NiagaraFloat=1
    )
    Particles.RibbonColor3:LinearColor=&VC_DynamicScaleColor(
     Color to Scale:LinearColor=&MakeCustomLinearColorFromBool(Bool=$User.bIsPainted, True Color=$Particles.Color, False Color=<rgba 1 1 0 1>)
     RGB Scale:Vector3f=&VectorFromFloat(Value=1)
     Alpha Scale:NiagaraFloat=1
    )
    Particles.RibbonColor4:LinearColor=&VC_DynamicScaleColor(
     Color to Scale:LinearColor=&MakeCustomLinearColorFromBool(Bool=$User.bIsPainted, True Color=$Particles.R1 Color, False Color=<rgba 1 0.12 0 1>)
     RGB Scale:Vector3f=&VectorFromFloat(Value=1)
     Alpha Scale:NiagaraFloat=1
    )
    Particles.RibbonWidth1:NiagaraFloat=&Multiply_Float(
     A=$Particles.WidthTag1
     B=&FloatFromCurve(FloatCurve=DataInterface() {propertyValues {Curve {keys [
          {interpMode=RCIM_Cubic, leaveTangentWeight=0.1, tangentMode=RCTM_User, time=0, value=0.6}
          {arriveTangentWeight=0.1166, interpMode=RCIM_Cubic, leaveTangentWeight=0.2721, tangentMode=RCTM_User, time=0.3, value=1}
          <at 1 1>
         ]}}})
    )
    Particles.RibbonWidth2:NiagaraFloat=&Multiply_Float(
     A=$Particles.WidthTag2
     B=&FloatFromCurve(FloatCurve=DataInterface() {propertyValues {Curve {keys [
          {interpMode=RCIM_Cubic, leaveTangentWeight=0.1, tangentMode=RCTM_User, time=0, value=0.6}
          {arriveTangentWeight=0.1166, interpMode=RCIM_Cubic, leaveTangentWeight=0.2721, tangentMode=RCTM_User, time=0.3, value=1}
          <at 1 1>
         ]}}})
    )
    Particles.RibbonWidth3:NiagaraFloat=&Multiply_Float(
     A=$Particles.WidthTag3
     B=&FloatFromCurve(FloatCurve=DataInterface() {propertyValues {Curve {keys [
          <at 0.085 0 RCIM_Constant>
          {arriveTangent=5.7357, interpMode=RCIM_Cubic, leaveTangent=5.7357, time=0.1158, value=0.3844}
          {arriveTangentWeight=0.1213, interpMode=RCIM_Cubic, leaveTangentWeight=0.2163, tangentMode=RCTM_User, time=0.2593, value=1}
          {arriveTangentWeight=0.0876, interpMode=RCIM_Cubic, leaveTangentWeight=0.3393, tangentMode=RCTM_User, time=0.5, value=1}
          {arriveTangent=-3.5379, interpMode=RCIM_Cubic, leaveTangent=-3.5379, time=0.7679, value=0.2867}
          <at 0.7827 0 RCIM_Constant>
         ]}}})
    )
   ) {isSetParameters=true, module=SetVariables_#3}
  ]
  rendererClasses [&NiagaraRibbonRendererProperties]
  renderers [
   {rendererClass=&NiagaraRibbonRendererProperties}
   {rendererClass=&NiagaraRibbonRendererProperties, rendererIndex=1}
   {rendererClass=&NiagaraRibbonRendererProperties, rendererIndex=2}
   {rendererClass=&NiagaraRibbonRendererProperties, rendererIndex=3}
   {rendererClass=&NiagaraRibbonRendererProperties, rendererIndex=4}
  ]
 }
}
summary {
 systemName=NS_Boost_CrazyShapes
 emitters [
  {emitterName=CrazyShapes, rendererClasses [&NiagaraSpriteRendererProperties]}
  {emitterName=MainRibbon, rendererClasses [&NiagaraRibbonRendererProperties]}
  {bEnabled=false, emitterName=MainRibbon001, rendererClasses [&NiagaraRibbonRendererProperties]}
 ]
 userVariables [
  User.bIsFrontendPreview {type=NiagaraBool}
  User.bIsPainted {type=NiagaraBool}
  User.ColorVariantID {type=NiagaraInt32}
  User.PaintedColor {type=LinearColor, description="", defaultValue {struct=LinearColor, value=<rgba 1 1 1 1>}}
  User.PaintedScalar {type=NiagaraFloat, defaultValue=1, description=""}
  User.PortCount {type=NiagaraInt32, defaultValue=1}
 ]
}
